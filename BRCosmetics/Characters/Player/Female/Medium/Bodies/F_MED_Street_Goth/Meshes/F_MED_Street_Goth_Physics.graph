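# Physics asset for the F_MED_Street_Goth female body, giving its skirt bones bodies and constraints for secondary motion.
# Capsules on pelvis, spine_03, thighs and calves form body collision shapes, joined by free-rotation constraints.
# Box bodies form front, side and back skirt chains rooted at the pelvis, with tight twist limits.
# Cross-constraints tie front side panels to outer panels, and each _04 bone constrains to _02, skipping _03.
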PhysicsAsset /BRCosmetics/Characters/Player/Female/Medium/Bodies/F_MED_Street_Goth/Meshes/F_MED_Street_Goth_Physics
bodies {
 calf_l [{shapeType=Capsule, length=29.3752, radius=5.4592, center=<-14.0706 -3.5012 -0.4484>, rotation=<rot 0 100.888 -85.8593>}]
 calf_r [{shapeType=Capsule, length=29.3752, radius=5.4592, center=<14.0706 3.5012 0.4484>, rotation=<rot 0 100.888 -85.8593>}]
 dyn_skirt_bk_01 [{shapeType=Box, extentX=17.4378, extentY=2.4902, extentZ=11.5177, center=<0.4038 0 -6.8425>, rotation=<rot 0 89.9999 5.2651>}]
 dyn_skirt_bk_02 [{shapeType=Box, extentX=10, extentY=2, extentZ=13.9797, center=<0 0 -6.8575>, rotation=<rot 0 90 0>}]
 dyn_skirt_bk_03 [{shapeType=Box, extentX=10, extentY=2, extentZ=13.9797, center=<0 0 -6.8575>, rotation=<rot 0 90 0>}]
 dyn_skirt_bk_04 [{shapeType=Box, extentX=10, extentY=2, extentZ=13.9797, center=<0 0 -6.8575>, rotation=<rot 0 90 0>}]
 dyn_skirt_bk_l_01 [{shapeType=Box, extentX=10, extentY=2, extentZ=13.9797, center=<0 0 -6.8575>, rotation=<rot 0 90 0>}]
 dyn_skirt_bk_l_02 [{shapeType=Box, extentX=10, extentY=2, extentZ=13.9797, center=<0 0 -6.8575>, rotation=<rot 0 90 0>}]
 dyn_skirt_bk_l_03 [{shapeType=Box, extentX=10, extentY=2, extentZ=13.9797, center=<0 0 -6.8575>, rotation=<rot 0 90 0>}]
 dyn_skirt_bk_l_04 [{shapeType=Box, extentX=10, extentY=2, extentZ=13.9797, center=<0 0 -6.8575>, rotation=<rot 0 90 0>}]
 dyn_skirt_bk_r_01 [{shapeType=Box, extentX=10, extentY=2, extentZ=13.9797, center=<0 0 6.8575>, rotation=<rot 0 90 0>}]
 dyn_skirt_bk_r_02 [{shapeType=Box, extentX=10, extentY=2, extentZ=13.9797, center=<0 0 6.8575>, rotation=<rot 0 90 0>}]
 dyn_skirt_bk_r_03 [{shapeType=Box, extentX=10, extentY=2, extentZ=13.9797, center=<0 0 6.8575>, rotation=<rot 0 90 0>}]
 dyn_skirt_bk_r_04 [{shapeType=Box, extentX=10, extentY=2, extentZ=13.9797, center=<0 0 6.8575>, rotation=<rot 0 90 0>}]
 dyn_skirt_fr_01 [{shapeType=Box, extentX=10, extentY=2, extentZ=13.9797, center=<0 0 -6.8575>, rotation=<rot 0 90 0>}]
 dyn_skirt_fr_l_01 [{shapeType=Box, extentX=14.4902, extentY=3.3418, extentZ=8.5705, center=<0.9549 1.2036 -10.4615>, rotation=<rot -3.5996 100.5854 18.5726>}]
 dyn_skirt_fr_l_02 [{shapeType=Box, extentX=14.4902, extentY=3.3418, extentZ=8.5705, center=<4.298 2.2392 -3.1285>, rotation=<rot 1.9466 102.4477 18.5461>}]
 dyn_skirt_fr_r_01 [{shapeType=Box, extentX=14.4902, extentY=3.3418, extentZ=8.5705, center=<-0.9549 -1.2036 10.4615>, rotation=<rot -3.5996 100.5854 18.5726>}]
 dyn_skirt_fr_r_02 [{shapeType=Box, extentX=14.4902, extentY=3.3418, extentZ=8.5705, center=<-4.298 -2.2392 3.1285>, rotation=<rot 1.9466 102.4477 18.5461>}]
 dyn_skirt_out_l_01 [{shapeType=Box, extentX=10, extentY=2, extentZ=13.9797, center=<0 0 -6.8575>, rotation=<rot 0 90 0>}]
 dyn_skirt_out_l_02 [{shapeType=Box, extentX=10, extentY=2, extentZ=13.9797, center=<0 0 -6.8575>, rotation=<rot 0 90 0>}]
 dyn_skirt_out_l_03 [{shapeType=Box, extentX=10, extentY=2, extentZ=13.9797, center=<0 0 -6.8575>, rotation=<rot 0 90 0>}]
 dyn_skirt_out_l_04 [{shapeType=Box, extentX=10, extentY=2, extentZ=13.9797, center=<0 0 -6.8575>, rotation=<rot 0 90 0>}]
 dyn_skirt_out_r_01 [{shapeType=Box, extentX=10, extentY=2, extentZ=13.9797, center=<0 0 6.8575>, rotation=<rot 0 90 0>}]
 dyn_skirt_out_r_02 [{shapeType=Box, extentX=10, extentY=2, extentZ=13.9797, center=<0 0 6.8575>, rotation=<rot 0 90 0>}]
 dyn_skirt_out_r_03 [{shapeType=Box, extentX=10, extentY=2, extentZ=13.9797, center=<0 0 6.8575>, rotation=<rot 0 90 0>}]
 dyn_skirt_out_r_04 [{shapeType=Box, extentX=10, extentY=2, extentZ=13.9797, center=<0 0 6.8575>, rotation=<rot 0 90 0>}]
 pelvis [{shapeType=Capsule, length=5, radius=5}]
 spine_03 [{shapeType=Capsule, length=7.7864, radius=8.9465, center=<-0.33 2.7639 0>, rotation=<rot 90 0 -8.44>}]
 thigh_l [
  {shapeType=Capsule, length=29.4416, radius=7.5506, center=<-22.6543 0.9069 0.7864>, rotation=<rot -0.1691 85.9191 -92.1248>}
  {shapeType=Capsule, length=6.5051, radius=4.6446, center=<-18.8281 3.4429 -2.7664>, rotation=<rot -90 0.0198 -0.0199>}
 ]
 thigh_r [
  {shapeType=Capsule, length=29.4416, radius=7.5506, center=<22.6543 -0.9069 -0.7864>, rotation=<rot -0.1691 85.9191 -92.1248>}
  {shapeType=Capsule, length=6.5051, radius=4.6446, center=<18.8281 -3.4429 2.7664>, rotation=<rot -90 0.0198 -0.0199>}
 ]
}
constraints [
 {
  bone1Name=calf_l
  bone2Name=thigh_l
  swing1LimitDegrees=45
  swing1Motion=Free
  swing2LimitDegrees=45
  swing2Motion=Free
  twistLimitDegrees=45
  twistMotion=Free
 }
 {
  bone1Name=calf_r
  bone2Name=thigh_r
  swing1LimitDegrees=45
  swing1Motion=Free
  swing2LimitDegrees=45
  swing2Motion=Free
  twistLimitDegrees=45
  twistMotion=Free
 }
 {
  bone1Name=thigh_l
  bone2Name=pelvis
  swing1LimitDegrees=45
  swing1Motion=Free
  swing2LimitDegrees=45
  swing2Motion=Free
  twistLimitDegrees=45
  twistMotion=Free
 }
 {
  bone1Name=thigh_r
  bone2Name=pelvis
  swing1LimitDegrees=45
  swing1Motion=Free
  swing2LimitDegrees=45
  swing2Motion=Free
  twistLimitDegrees=45
  twistMotion=Free
 }
 {bone1Name=dyn_skirt_fr_01, bone2Name=pelvis, swing1LimitDegrees=45, swing2LimitDegrees=45, twistLimitDegrees=2}
 {bone1Name=dyn_skirt_out_l_01, bone2Name=pelvis, swing1LimitDegrees=45, swing2LimitDegrees=20, twistLimitDegrees=2}
 {bone1Name=dyn_skirt_out_l_02, bone2Name=dyn_skirt_out_l_01, swing1LimitDegrees=45, swing2LimitDegrees=20, twistLimitDegrees=2}
 {bone1Name=dyn_skirt_out_l_03, bone2Name=dyn_skirt_out_l_02, swing1LimitDegrees=45, swing2LimitDegrees=20, twistLimitDegrees=2}
 {bone1Name=dyn_skirt_out_l_04, bone2Name=dyn_skirt_out_l_02, swing1LimitDegrees=1, swing2LimitDegrees=20, twistLimitDegrees=1}
 {bone1Name=dyn_skirt_bk_01, bone2Name=pelvis, swing1LimitDegrees=20, swing2LimitDegrees=20, twistLimitDegrees=10}
 {bone1Name=dyn_skirt_bk_02, bone2Name=dyn_skirt_bk_01, swing1LimitDegrees=20, swing2LimitDegrees=20, twistLimitDegrees=2}
 {bone1Name=dyn_skirt_bk_03, bone2Name=dyn_skirt_bk_02, swing1LimitDegrees=20, swing2LimitDegrees=20, twistLimitDegrees=2}
 {bone1Name=dyn_skirt_bk_04, bone2Name=dyn_skirt_bk_02, swing1LimitDegrees=1, swing2LimitDegrees=20, twistLimitDegrees=1}
 {bone1Name=dyn_skirt_fr_l_01, bone2Name=pelvis, swing1LimitDegrees=3, swing2LimitDegrees=45, twistLimitDegrees=2}
 {bone1Name=dyn_skirt_fr_l_02, bone2Name=dyn_skirt_fr_l_01, swing1LimitDegrees=45, swing2LimitDegrees=45, twistLimitDegrees=2}
 {bone1Name=dyn_skirt_bk_l_01, bone2Name=pelvis, swing1LimitDegrees=20, swing2LimitDegrees=20, twistLimitDegrees=2}
 {bone1Name=dyn_skirt_bk_l_02, bone2Name=dyn_skirt_bk_l_01, swing1LimitDegrees=20, swing2LimitDegrees=20, twistLimitDegrees=2}
 {bone1Name=dyn_skirt_bk_l_03, bone2Name=dyn_skirt_bk_l_02, swing1LimitDegrees=20, swing2LimitDegrees=20, twistLimitDegrees=2}
 {bone1Name=dyn_skirt_bk_l_04, bone2Name=dyn_skirt_bk_l_02, swing1LimitDegrees=1, swing2LimitDegrees=20, twistLimitDegrees=1}
 {bone1Name=dyn_skirt_fr_r_01, bone2Name=pelvis, swing1LimitDegrees=3, swing2LimitDegrees=45, twistLimitDegrees=2}
 {bone1Name=dyn_skirt_fr_r_02, bone2Name=dyn_skirt_fr_r_01, swing1LimitDegrees=45, swing2LimitDegrees=45, twistLimitDegrees=2}
 {bone1Name=dyn_skirt_out_r_01, bone2Name=pelvis, swing1LimitDegrees=45, swing2LimitDegrees=20, twistLimitDegrees=2}
 {bone1Name=dyn_skirt_out_r_02, bone2Name=dyn_skirt_out_r_01, swing1LimitDegrees=45, swing2LimitDegrees=20, twistLimitDegrees=2}
 {bone1Name=dyn_skirt_out_r_03, bone2Name=dyn_skirt_out_r_02, swing1LimitDegrees=45, swing2LimitDegrees=20, twistLimitDegrees=2}
 {bone1Name=dyn_skirt_out_r_04, bone2Name=dyn_skirt_out_r_02, swing1LimitDegrees=1, swing2LimitDegrees=20, twistLimitDegrees=1}
 {bone1Name=dyn_skirt_bk_r_01, bone2Name=pelvis, swing1LimitDegrees=20, swing2LimitDegrees=20, twistLimitDegrees=2}
 {bone1Name=dyn_skirt_bk_r_02, bone2Name=dyn_skirt_bk_r_01, swing1LimitDegrees=20, swing2LimitDegrees=20, twistLimitDegrees=2}
 {bone1Name=dyn_skirt_bk_r_03, bone2Name=dyn_skirt_bk_r_02, swing1LimitDegrees=20, swing2LimitDegrees=20, twistLimitDegrees=2}
 {bone1Name=dyn_skirt_bk_r_04, bone2Name=dyn_skirt_bk_r_02, swing1LimitDegrees=1, swing2LimitDegrees=20, twistLimitDegrees=1}
 {bone1Name=dyn_skirt_fr_l_02, bone2Name=dyn_skirt_out_l_02, swing1LimitDegrees=45, swing2LimitDegrees=20, twistLimitDegrees=2}
 {bone1Name=dyn_skirt_fr_r_02, bone2Name=dyn_skirt_out_r_02, swing1LimitDegrees=45, swing2LimitDegrees=20, twistLimitDegrees=2}
 {
  bone1Name=spine_03
  bone2Name=pelvis
  swing1LimitDegrees=45
  swing1Motion=Free
  swing2LimitDegrees=45
  swing2Motion=Free
  twistLimitDegrees=45
  twistMotion=Free
 }
]
