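# Physics asset for the strings on the F_MED_Street_Goth female player cosmetic body, covering left and right dynamic string bone chains.
# Spheres at the string roots and small capsules below form a three-bone left chain and four-bone right chain, linked by angle-limited constraints with tight twist.
# Larger pelvis and spine capsules form a torso volume, joined by free-rotation constraints.
# No constraint links either string root to the torso.
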PhysicsAsset /BRCosmetics/Characters/Player/Female/Medium/Bodies/F_MED_Street_Goth/Meshes/F_MED_Street_Goth_String_Physics
bodies {
 dyn_string_l_01 [{shapeType=Sphere, radius=1, shapeName=dyn_string_l_01_sphere}]
 dyn_string_l_02 [
  {shapeType=Capsule, length=2, radius=1, center=<0.76 0 0>, rotation=<rot 90 0 0>}
  {shapeType=Capsule, length=3, radius=0.5, shapeName=dyn_string_l_02_capsule, center=<-4.9669 0 -1.2784>, rotation=<rot 69.9999 -180 180>}
 ]
 dyn_string_l_03 [{shapeType=Capsule, length=1, radius=1, shapeName=dyn_string_l_03_capsule, center=<1.5081 0 0>, rotation=<rot 90 0 0>}]
 dyn_string_r_01 [{shapeType=Sphere, radius=1, shapeName=dyn_string_r_01_sphere}]
 dyn_string_r_02 [{shapeType=Capsule, length=2, radius=1.5, center=<1 0 0>, rotation=<rot -90 0 0>}]
 dyn_string_r_03 [{shapeType=Capsule, length=1.5, radius=1.5, center=<1 0 0>, rotation=<rot 90 0 0>}]
 dyn_string_r_04 [{shapeType=Capsule, length=1.5, radius=1.5, center=<1 0 0>, rotation=<rot 90 0 0>}]
 pelvis [{shapeType=Capsule, length=23.7645, radius=11.2575, center=<-1.3945 -0.9753 0.2775>, rotation=<rot 0 -90 0>}]
 spine_02 [{shapeType=Capsule, length=12.8892, radius=8.6529, center=<1.4234 0.374 -0.0991>, rotation=<rot 0 -90 0>}]
 spine_05 [{shapeType=Capsule, length=19.5, radius=10.5, center=<-6.2484 0 0>}]
}
constraints [
 {bone1Name=dyn_string_r_03, bone2Name=dyn_string_r_02, swing1LimitDegrees=45, swing2LimitDegrees=29.1429, twistLimitDegrees=5}
 {
  bone1Name=spine_05
  bone2Name=spine_02
  swing1LimitDegrees=45
  swing1Motion=Free
  swing2LimitDegrees=45
  swing2Motion=Free
  twistLimitDegrees=45
  twistMotion=Free
 }
 {
  bone1Name=spine_02
  bone2Name=pelvis
  swing1LimitDegrees=45
  swing1Motion=Free
  swing2LimitDegrees=45
  swing2Motion=Free
  twistLimitDegrees=45
  twistMotion=Free
 }
 {bone1Name=dyn_string_r_04, bone2Name=dyn_string_r_03, swing1LimitDegrees=45, swing2LimitDegrees=29.1429, twistLimitDegrees=5}
 {bone1Name=dyn_string_r_02, bone2Name=dyn_string_r_01, swing1LimitDegrees=5, swing2LimitDegrees=45, twistLimitDegrees=5}
 {bone1Name=dyn_string_l_02, bone2Name=dyn_string_l_01, swing1LimitDegrees=10, swing2LimitDegrees=45, twistLimitDegrees=5}
 {bone1Name=dyn_string_l_03, bone2Name=dyn_string_l_02, swing1LimitDegrees=45, swing2LimitDegrees=20, twistLimitDegrees=5}
]
