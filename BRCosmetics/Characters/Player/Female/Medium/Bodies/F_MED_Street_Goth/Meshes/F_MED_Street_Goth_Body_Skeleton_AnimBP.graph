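# Secondary-motion AnimBP for the F_MED_Street_Goth female body, simulating skirt, strings, bows and bracelet.
# It copies the parent mesh pose, then chains bone constraints, RigidBody and AnimDynamics nodes, averaging left/right front skirt bones into center ones.
# The EventGraph derives skydiving force from head-to-pelvis direction and dive mode, feeding skirt physics.
# String simulation alpha drops as DeltaTime rises; crouching and the ShaolinSipUp emote trigger additive skirt offsets.
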AnimBlueprint /BRCosmetics/Characters/Player/Female/Medium/Bodies/F_MED_Street_Goth/Meshes/F_MED_Street_Goth_Body_Skeleton_AnimBP
target_skeleton=&/F_MED_Street_Goth_Body_Skeleton
graphs [
 AnimGraph {nodes [
   n1 = Get active_skydiving vector force() -> (active_skydiving vector force:Vector)
   n2 = "Constraint - dyn_string_r_01 (PPP)"!(
    ComponentPose=Local To Component(LocalPose=Copy Pose From Mesh() {bUseAttachedParent=true})
    Alpha=1
    ConstraintWeights_0=0.45
    ConstraintWeights_1:float=0.5103
    ConstraintWeights_2:float=0.0397
   ) -> (Pose) {
    alphaBoolBlend {blendOption=Linear}
    boneToModify {boneName=dyn_string_r_01}
    constraintSetup [{targetBone {boneName=spine_04}}, {targetBone {boneName=spine_05}}, {targetBone {boneName=clavicle_r}}]
    constraintWeights [1,1,1]
   }
   n3 = "Constraint - dyn_string_l_01 (PPP)"!(ComponentPose=@n2, Alpha=1, ConstraintWeights_0=0.6547, ConstraintWeights_1:float=0.2849, ConstraintWeights_2:float=0.0604) -> (Pose) {
    alphaBoolBlend {blendOption=Linear}
    boneToModify {boneName=dyn_string_l_01}
    constraintSetup [{targetBone {boneName=spine_04}}, {targetBone {boneName=spine_05}}, {targetBone {boneName=clavicle_l}}]
    constraintWeights [1,1,1]
   }
   n4 = RigidBody(ComponentPose=@n3, ExternalForce=@n1) {
    bOverrideWorldGravity=true
    overridePhysicsAsset=&/F_MED_Street_Goth_Physics
    simulationSpace=BaseBoneSpace
    alphaBoolBlend {blendOption=Linear}
    baseBoneRef {boneName=root}
    componentAppliedLinearAccClamp=<1000 1000 1000>
    componentLinearAccScale {z=0.5}
    componentLinearVelScale {z=0.5}
    overrideWorldGravity {z=-980}
   }
   n5 = "Constraint - dyn_skirt_fr_l_01 (P)"(ComponentPose=@n4, Alpha=1, ConstraintWeights_0=1) {
    alphaBoolBlend {blendOption=Linear}
    boneToModify {boneName=dyn_skirt_fr_l_01}
    constraintSetup [{targetBone {boneName=thigh_l}}]
    constraintWeights [1]
   }
   n6 = "Constraint - dyn_skirt_fr_r_01 (P)"(ComponentPose=@n5, Alpha=1, ConstraintWeights_0=1) {
    alphaBoolBlend {blendOption=Linear}
    boneToModify {boneName=dyn_skirt_fr_r_01}
    constraintSetup [{targetBone {boneName=thigh_r}}]
    constraintWeights [1]
   }
   n7 = "Constraint - dyn_skirt_out_l_01 (P)"(ComponentPose=@n6, Alpha=1, ConstraintWeights_0=1) {
    alphaBoolBlend {blendOption=Linear}
    boneToModify {boneName=dyn_skirt_out_l_01}
    constraintSetup [{targetBone {boneName=thigh_l}}]
    constraintWeights [1]
   }
   n8 = "Constraint - dyn_skirt_out_r_01 (P)"(ComponentPose=@n7, Alpha=1, ConstraintWeights_0=1) {
    alphaBoolBlend {blendOption=Linear}
    boneToModify {boneName=dyn_skirt_out_r_01}
    constraintSetup [{targetBone {boneName=thigh_r}}]
    constraintWeights [1]
   }
   n9 = Anim Dynamics - Bone: dyn_skirt_fr_l_01(ComponentPose=@n8, Alpha=1, ExternalForce=@n1) {
    angularDampingOverride=0.98
    bGravityOverrideInSimSpace=true
    bOverrideAngularDamping=true
    bUseGravityOverride=true
    numSolverIterationsPostUpdate=2
    numSolverIterationsPreUpdate=8
    simulationSpace=BoneRelative
    alphaBoolBlend {blendOption=Linear}
    boundBone {boneName=dyn_skirt_fr_l_01}
    gravityOverride {z=-980}
    physicsBodyDefinitions [{
      sphereCollisionRadius=0
      boundBone {boneName=dyn_skirt_fr_l_01}
      boxExtents=<20 20 10>
      constraintSetup {twistAxis=AxisZ, angularLimitsMax=<0 0 2>, angularLimitsMin=<0 -30 -2>, angularTarget=<0 0 0>}
      localJointOffset=<0 0 -3>
     }]
    relativeSpaceBone {boneName=root}
   }
   n10 = Anim Dynamics - Bone: dyn_skirt_fr_l_02(ComponentPose=@n9, Alpha=1, ExternalForce=@n1) {
    angularDampingOverride=0.98
    bGravityOverrideInSimSpace=true
    bOverrideAngularDamping=true
    bUseGravityOverride=true
    simulationSpace=BoneRelative
    alphaBoolBlend {blendOption=Linear}
    boundBone {boneName=dyn_skirt_fr_l_02}
    gravityOverride {z=-980}
    physicsBodyDefinitions [{
      sphereCollisionRadius=0
      boundBone {boneName=dyn_skirt_fr_l_02}
      boxExtents=<20 20 10>
      constraintSetup {twistAxis=AxisZ, angularLimitsMax=<10 1 2>, angularLimitsMin=<-10 -30 -2>, angularTarget=<0 0 0>}
      localJointOffset=<0 0 -3>
     }]
    relativeSpaceBone {boneName=root}
   }
   n11 = Anim Dynamics - Bone: dyn_skirt_fr_l_03(ComponentPose=@n10, Alpha=1, ExternalForce=@n1) {
    angularDampingOverride=0.98
    bGravityOverrideInSimSpace=true
    bOverrideAngularDamping=true
    bUseGravityOverride=true
    simulationSpace=BoneRelative
    alphaBoolBlend {blendOption=Linear}
    boundBone {boneName=dyn_skirt_fr_l_03}
    gravityOverride {z=-980}
    physicsBodyDefinitions [{
      sphereCollisionRadius=0
      boundBone {boneName=dyn_skirt_fr_l_03}
      boxExtents=<20 20 10>
      constraintSetup {twistAxis=AxisZ, angularLimitsMax=<10 0 2>, angularLimitsMin=<-10 -30 -2>, angularTarget=<0 0 0>}
      localJointOffset=<0 0 -3>
     }]
    relativeSpaceBone {boneName=root}
   }
   n12 = Anim Dynamics - Bone: dyn_skirt_fr_l_04(ComponentPose=@n11, Alpha=1, ExternalForce=@n1) {
    angularDampingOverride=0.98
    bGravityOverrideInSimSpace=true
    bOverrideAngularDamping=true
    bUseGravityOverride=true
    simulationSpace=BoneRelative
    alphaBoolBlend {blendOption=Linear}
    boundBone {boneName=dyn_skirt_fr_l_04}
    gravityOverride {z=-980}
    physicsBodyDefinitions [{
      sphereCollisionRadius=0
      boundBone {boneName=dyn_skirt_fr_l_04}
      boxExtents=<20 20 10>
      constraintSetup {twistAxis=AxisZ, angularLimitsMax=<0 0 0>, angularLimitsMin=<0 -30 0>, angularTarget=<0 0 0>}
      localJointOffset=<0 0 -3>
     }]
    relativeSpaceBone {boneName=root}
   }
   n13 = Anim Dynamics - Bone: dyn_skirt_fr_r_01(ComponentPose=@n12, Alpha=1, ExternalForce=@n1) {
    angularDampingOverride=0.98
    bGravityOverrideInSimSpace=true
    bOverrideAngularDamping=true
    bUseGravityOverride=true
    simulationSpace=BoneRelative
    alphaBoolBlend {blendOption=Linear}
    boundBone {boneName=dyn_skirt_fr_r_01}
    gravityOverride {z=-980}
    physicsBodyDefinitions [{
      sphereCollisionRadius=0
      boundBone {boneName=dyn_skirt_fr_r_01}
      boxExtents=<20 20 10>
      constraintSetup {twistAxis=AxisZ, angularLimitsMax=<0 0 2>, angularLimitsMin=<0 -30 -2>, angularTarget=<0 0 0>}
      localJointOffset=<0 0 3>
     }]
    relativeSpaceBone {boneName=root}
   }
   n14 = Anim Dynamics - Bone: dyn_skirt_fr_r_02(ComponentPose=@n13, Alpha=1, ExternalForce=@n1) {
    angularDampingOverride=0.98
    bGravityOverrideInSimSpace=true
    bOverrideAngularDamping=true
    bUseGravityOverride=true
    simulationSpace=BoneRelative
    alphaBoolBlend {blendOption=Linear}
    boundBone {boneName=dyn_skirt_fr_r_02}
    gravityOverride {z=-980}
    physicsBodyDefinitions [{
      sphereCollisionRadius=0
      boundBone {boneName=dyn_skirt_fr_r_02}
      boxExtents=<20 20 10>
      constraintSetup {twistAxis=AxisZ, angularLimitsMax=<10 1 2>, angularLimitsMin=<-10 -30 -2>, angularTarget=<0 0 0>}
      localJointOffset=<0 0 3>
     }]
    relativeSpaceBone {boneName=root}
   }
   n15 = Anim Dynamics - Bone: dyn_skirt_fr_r_03(ComponentPose=@n14, Alpha=1, ExternalForce=@n1) {
    angularDampingOverride=0.98
    bGravityOverrideInSimSpace=true
    bOverrideAngularDamping=true
    bUseGravityOverride=true
    simulationSpace=BoneRelative
    alphaBoolBlend {blendOption=Linear}
    boundBone {boneName=dyn_skirt_fr_r_03}
    gravityOverride {z=-980}
    physicsBodyDefinitions [{
      sphereCollisionRadius=0
      boundBone {boneName=dyn_skirt_fr_r_03}
      boxExtents=<20 20 10>
      constraintSetup {twistAxis=AxisZ, angularLimitsMax=<10 3 2>, angularLimitsMin=<-10 -30 -2>, angularTarget=<0 0 0>}
      localJointOffset=<0 0 3>
     }]
    relativeSpaceBone {boneName=root}
   }
   n16 = Anim Dynamics - Bone: dyn_skirt_fr_r_04(ComponentPose=@n15, Alpha=1, ExternalForce=@n1) {
    angularDampingOverride=0.98
    bGravityOverrideInSimSpace=true
    bOverrideAngularDamping=true
    bUseGravityOverride=true
    simulationSpace=BoneRelative
    alphaBoolBlend {blendOption=Linear}
    boundBone {boneName=dyn_skirt_fr_r_04}
    gravityOverride {z=-980}
    physicsBodyDefinitions [{
      sphereCollisionRadius=0
      boundBone {boneName=dyn_skirt_fr_r_04}
      boxExtents=<20 20 10>
      constraintSetup {twistAxis=AxisZ, angularLimitsMax=<0 0 0>, angularLimitsMin=<0 -30 0>, angularTarget=<0 0 0>}
      localJointOffset=<0 0 3>
     }]
    relativeSpaceBone {boneName=root}
   }
   n17 = "Constraint - dyn_skirt_fr_01 (PP)"!(ComponentPose=@n16, Alpha=1, ConstraintWeights_0=0.5, ConstraintWeights_1:float=0.5) -> (Pose) {
    alphaBoolBlend {blendOption=Linear}
    boneToModify {boneName=dyn_skirt_fr_01}
    constraintSetup [{targetBone {boneName=dyn_skirt_fr_l_01}}, {targetBone {boneName=dyn_skirt_fr_r_01}}]
    constraintWeights [1,1]
   }
   n18 = "Constraint - dyn_skirt_fr_02 (RR)"!(ComponentPose=@n17, Alpha=1, ConstraintWeights_0=0.5, ConstraintWeights_1:float=0.5) -> (Pose) {
    alphaBoolBlend {blendOption=Linear}
    boneToModify {boneName=dyn_skirt_fr_02}
    constraintSetup [{transformType=Rotation, targetBone {boneName=dyn_skirt_fr_l_02}}, {transformType=Rotation, targetBone {boneName=dyn_skirt_fr_r_02}}]
    constraintWeights [1,1]
   }
   n19 = "Constraint - dyn_skirt_fr_03 (RR)"!(ComponentPose=@n18, Alpha=1, ConstraintWeights_0=0.5, ConstraintWeights_1:float=0.5) -> (Pose) {
    alphaBoolBlend {blendOption=Linear}
    boneToModify {boneName=dyn_skirt_fr_03}
    constraintSetup [{transformType=Rotation, targetBone {boneName=dyn_skirt_fr_l_03}}, {transformType=Rotation, targetBone {boneName=dyn_skirt_fr_r_03}}]
    constraintWeights [1,1]
   }
   n20 = "Constraint - dyn_skirt_fr_04 (RR)"!(ComponentPose=@n19, Alpha=1, ConstraintWeights_0=0.5, ConstraintWeights_1:float=0.5) -> (Pose) {
    alphaBoolBlend {blendOption=Linear}
    boneToModify {boneName=dyn_skirt_fr_04}
    constraintSetup [{transformType=Rotation, targetBone {boneName=dyn_skirt_fr_l_04}}, {transformType=Rotation, targetBone {boneName=dyn_skirt_fr_r_04}}]
    constraintWeights [1,1]
   }
   n21 = Anim Dynamics - Bone: dyn_bow_leg(ComponentPose=@n20, Alpha=0.5, ExternalForce="0,0,0") {
    angularDampingOverride=0.97
    bGravityOverrideInSimSpace=true
    bOverrideAngularDamping=true
    gravityScale=0.65
    lODThreshold=1
    linearDampingOverride=0.97
    numSolverIterationsPostUpdate=4
    numSolverIterationsPreUpdate=16
    simulationSpace=BoneRelative
    alphaBoolBlend {blendOption=Linear}
    boundBone {boneName=dyn_bow_leg}
    componentLinearAccScale=<0.5 0.5 0.5>
    componentLinearVelScale=<0.5 0.5 0.5>
    physicsBodyDefinitions [{
      sphereCollisionRadius=0
      boundBone {boneName=dyn_bow_leg}
      boxExtents=<5 5 5>
      constraintSetup {angularTargetAxis=AxisZ, twistAxis=AxisZ, angularLimitsMax=<0 30 30>, angularLimitsMin=<0 -30 -30>, angularTarget=<0 0 0>}
      localJointOffset=<0 0 -0.1>
     }]
    relativeSpaceBone {boneName=root}
   }
   n22 = Anim Dynamics - Bone: dyn_bow(ComponentPose=@n21, Alpha=0.35, ExternalForce="0,0,0") {
    angularDampingOverride=0.97
    bGravityOverrideInSimSpace=true
    bOverrideAngularDamping=true
    gravityScale=0.65
    lODThreshold=1
    linearDampingOverride=0.97
    numSolverIterationsPostUpdate=4
    numSolverIterationsPreUpdate=16
    simulationSpace=BoneRelative
    alphaBoolBlend {blendOption=Linear}
    boundBone {boneName=dyn_bow}
    componentLinearAccScale=<0.5 0.5 0.5>
    componentLinearVelScale=<0.5 0.5 0.5>
    physicsBodyDefinitions [{
      sphereCollisionRadius=0
      boundBone {boneName=dyn_bow}
      boxExtents=<5 5 5>
      constraintSetup {angularTargetAxis=AxisZ, twistAxis=AxisZ, angularLimitsMax=<30 30 30>, angularLimitsMin=<-30 -30 -30>, angularTarget=<0 0 0>}
      localJointOffset=<0 0 -0.1>
     }]
    relativeSpaceBone {boneName=root}
   }
   n23 = Anim Dynamics - Bone: dyn_bracelet(ComponentPose=@n22, Alpha=0.1, ExternalForce="0,0,0") {
    angularDampingOverride=0.97
    bGravityOverrideInSimSpace=true
    bOverrideAngularDamping=true
    lODThreshold=1
    linearDampingOverride=0.97
    numSolverIterationsPostUpdate=4
    numSolverIterationsPreUpdate=16
    simulationSpace=BoneRelative
    alphaBoolBlend {blendOption=Linear}
    boundBone {boneName=dyn_bracelet}
    physicsBodyDefinitions [{
      sphereCollisionRadius=0
      boundBone {boneName=dyn_bracelet}
      boxExtents=<5 5 5>
      constraintSetup {angularTargetAxis=AxisZ, twistAxis=AxisZ, angularLimitsMax=<45 0 10>, angularLimitsMin=<-45 -30 -10>, angularTarget=<0 0 0>}
      localJointOffset=<0 0 -0.1>
     }]
    relativeSpaceBone {boneName=root}
   }
   n24 = RigidBody!(ComponentPose=@n23, Alpha=Get DeltaTime() -> (DeltaTime:float), OverrideWorldGravity:Vector=Get Gravity Override(), ExternalForce) -> (Pose) {
    bClampLinearTranslationLimitToRefPose=true
    bOverrideWorldGravity=true
    overridePhysicsAsset=&/F_MED_Street_Goth_String_Physics
    simulationSpace=BaseBoneSpace
    alphaBoolBlend {blendOption=Linear}
    alphaScaleBiasClamp {bInterpResult=true, bMapRange=true, inRange {max=0.05, min=0.02}, outRange {max=0.2, min=1}}
    baseBoneRef {boneName=root}
    componentLinearAccScale=<1 1 1>
    componentLinearVelScale=<1 1 1>
    overrideWorldGravity {z=-980}
   }
   n25 = "Transform (Modify) Bone - Bone: dyn_skirt_fr_l_01"!(
    ComponentPose=@n24
    bAlphaBoolEnabled:bool=Get bIsCrouching()
    Translation="6.893,4.159,-3.475"
    Rotation="-2.932,-7.714,0.175"
    Scale="1,1,1"
   ) -> (Pose) {
    alphaInputType=Bool
    rotationMode=BMM_Additive
    rotationSpace=BCS_BoneSpace
    translationMode=BMM_Additive
    translationSpace=BCS_BoneSpace
    alphaBoolBlend {blendOption=Linear}
    boneToModify {boneName=dyn_skirt_fr_l_01}
    rotation=<rot -2.932 -7.7142 0.175>
    translation=<6.893 4.159 -3.475>
   }
   n26 = "Transform (Modify) Bone - Bone: dyn_skirt_out_l_01"(
    ComponentPose=@n25
    Alpha=Get emote_shaolin_sip_active() -> (emote_shaolin_sip_active:double)
    Translation="-0.873,3.167,6.169"
    Rotation="20.687,72.947,-9.443"
    Scale="1,1,1"
   ) {
    rotationMode=BMM_Additive
    rotationSpace=BCS_BoneSpace
    translationMode=BMM_Additive
    translationSpace=BCS_BoneSpace
    alphaBoolBlend {blendOption=Linear}
    alphaScaleBiasClamp {bInterpResult=true, interpSpeedIncreasing=5}
    boneToModify {boneName=dyn_skirt_out_l_01}
    rotation=<rot 20.687 72.947 -9.443>
    translation=<-0.873 3.167 6.169>
   }
   Output Pose(Result=Component To Local(ComponentPose=@n26)) {layerGroup=None}
  ]}
 EventGraph {nodes [
   n1 = Event BlueprintUpdateAnimation()
   > Set DeltaTime(DeltaTime:float=@n1.DeltaTimeX) -> (Output_Get:float)
   > Get!(execute:Exec) -> (then:Exec, else:Exec, Skel_Mesh:Skeletal Mesh Component Object Reference)
   > Get!(execute) -> (then, else, Anim_Inst:Anim Instance Object Reference)
   > n2 = Sequence!(execute) -> (then_0, then_1, then_2:Exec, then_3:Exec)
   n3 = Gravity Override State Function DEPRECATED(
    execute=@n2.then_3
    CharacterPartAnimInstance=Self-Reference()
    GameTimeInSeconds=GetGameTimeInSeconds()
    GravityOverrideParams=Get Gravity Override Params()
   )
   > Set Gravity Override(Gravity Override=@n3)
   n4 = Branch(
    execute=[@n12.then, @n2.then_0]
    Condition=Montage_IsPlaying(
     self=Get Anim_Inst() -> (Anim_Inst:Anim Instance Object Reference)
     Montage=/BRCosmetics/Animation/Game/MainPlayer/Emotes/ShaolinSipUp/Emote_ShaolinSipUp_CMF_M
    )
   )
   > Set emote_shaolin_sip_active(emote_shaolin_sip_active:double=1) -> (Output_Get:double)
   Set emote_shaolin_sip_active(execute=@n4.else, emote_shaolin_sip_active=0)
   n5 = Branch(execute=@n2.then_1, Condition=Get bCanPlayCustomAnimations() -> (bCanPlayCustomAnimations:bool))
   n6 = Branch(execute=[@n5.else, @n5.then], Condition=Get bIsSkyDiving())
   n7 = Branch(execute=[@n6.else, @n6.then], Condition=Get bIsSkydiveFloating())
   n8 = Get Direction Between Sockets!(Component:Skeletal Mesh Component Object Reference=Get Skel_Mesh(), SocketOrBoneNameFrom:Name=head, SocketOrBoneNameTo:Name=pelvis) -> (ReturnValue:Vector) {is=title K2Node_CallFunction}
   > Set skydiving vector force(skydiving vector force:Vector=vector + vector(
     A:Vector=vector * vector(A:Vector=@n8, B:Vector="75,75,75") -> (ReturnValue:Vector) {is=title K2Node_CommutativeAssociativeBinaryOperator}
     B:Vector="0, 0,2000"
    ) -> (ReturnValue:Vector) {is=title K2Node_CommutativeAssociativeBinaryOperator}) -> (Output_Get:Vector)
   n9 = Branch(execute=@n7.else, Condition=Get bIsSkydiveDiveMode())
   > Set skydiving vector force(skydiving vector force=vector + vector(A=vector * vector(A=@n8, B="180,180,200"), B="0, 0,2000"))
   Set skydiving vector force(execute=@n9.else, skydiving vector force=vector * vector(A=@n8, B="-15,-15,-15"))
   n10 = Branch(execute=@n2.then_2, Condition=Get bIsSkyDiving())
   > Set active_skydiving vector force(active_skydiving vector force:Vector=Get skydiving vector force() -> (skydiving vector force:Vector)) -> (Output_Get:Vector)
   Set active_skydiving vector force(execute=@n10.else, active_skydiving vector force="0, 0, 0")
   Event BlueprintInitializeAnimation()
   > Set Skel_Mesh(Skel_Mesh=GetOwnerSkeletalMeshComponent())
   > n11 = Get!(execute) -> (then, else, Skel_Mesh:Skeletal Mesh Component Object Reference)
   > Set Anim_Inst(Anim_Inst:Anim Instance Object Reference=GetAnimInstance(self=@n11)) -> (Output_Get:Anim Instance Object Reference)
   n12 = Branch(Condition=Get bCanPlayCustomAnimations())
  ]}
]
variables [
 active_skydiving vector force {type=Vector}
 Anim_Inst {type=AnimInstance, default=null}
 DeltaTime {type=number}
 emote_shaolin_sip_active {type=number}
 Gravity Override {type=Vector}
 Gravity Override Params {type=GravityOverrideParamsStruct}
 Skel_Mesh {type=SkeletalMeshComponent}
 skydiving vector force {type=Vector}
]
