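# Physics asset for the face accessory mesh of the female medium Street Goth player body, setting up physics for its left and right ponytails.
# Each ponytail is a four-bone chain of capsule bodies, constrained link by link from the head with swing limits and a tight twist limit.
# Capsules on the head and upper spine add collision shapes for the body, with the spine capsule unconstrained to the chains.
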PhysicsAsset /BRCosmetics/Characters/Player/Female/Medium/Bodies/F_MED_Street_Goth/Meshes/F_MED_Street_Goth_FaceACC_Physics
bodies {
 dyn_ponytail_l_01 [{shapeType=Capsule, length=0, radius=2.7763, center=<1 0 0>}]
 dyn_ponytail_l_02 [{shapeType=Capsule, length=5, radius=2.8218, center=<6 0 0>, rotation=<rot 90 0 0>}]
 dyn_ponytail_l_03 [{shapeType=Capsule, length=5, radius=2.0587, center=<6 0 0>, rotation=<rot 90 0 0>}]
 dyn_ponytail_l_04 [{shapeType=Capsule, length=0, radius=1.4421, center=<1 0 0>}]
 dyn_ponytail_r_01 [{shapeType=Capsule, length=0, radius=2.86, center=<1 0 0>}]
 dyn_ponytail_r_02 [{shapeType=Capsule, length=5, radius=2.7733, center=<6 0 0>, rotation=<rot 90 0 0>}]
 dyn_ponytail_r_03 [{shapeType=Capsule, length=5, radius=2.0904, center=<6 0 0>, rotation=<rot 90 0 0>}]
 dyn_ponytail_r_04 [{shapeType=Capsule, length=0, radius=1.4355, center=<1 0 0>}]
 head [{shapeType=Capsule, length=0, radius=5}]
 spine_05 [{shapeType=Capsule, length=44.75, radius=12.5, center=<0 1.3272 0>}]
}
constraints [
 {bone1Name=dyn_ponytail_r_01, bone2Name=head, swing1LimitDegrees=45, swing2LimitDegrees=45, twistLimitDegrees=5}
 {bone1Name=dyn_ponytail_r_02, bone2Name=dyn_ponytail_r_01, swing1LimitDegrees=45, swing2LimitDegrees=45, twistLimitDegrees=5}
 {bone1Name=dyn_ponytail_r_03, bone2Name=dyn_ponytail_r_02, swing1LimitDegrees=45, swing2LimitDegrees=45, twistLimitDegrees=5}
 {bone1Name=dyn_ponytail_r_04, bone2Name=dyn_ponytail_r_03, swing1LimitDegrees=45, swing2LimitDegrees=45, twistLimitDegrees=5}
 {bone1Name=dyn_ponytail_l_01, bone2Name=head, swing1LimitDegrees=45, swing2LimitDegrees=45, twistLimitDegrees=5}
 {bone1Name=dyn_ponytail_l_02, bone2Name=dyn_ponytail_l_01, swing1LimitDegrees=45, swing2LimitDegrees=45, twistLimitDegrees=5}
 {bone1Name=dyn_ponytail_l_03, bone2Name=dyn_ponytail_l_02, swing1LimitDegrees=45, swing2LimitDegrees=45, twistLimitDegrees=5}
 {bone1Name=dyn_ponytail_l_04, bone2Name=dyn_ponytail_l_03, swing1LimitDegrees=45, swing2LimitDegrees=45, twistLimitDegrees=5}
]
