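# Animation Blueprint for the F_MED_Street_Goth face accessory skeleton, adding secondary motion to bangs, ribbons and ponytails.
# It copies the attached parent mesh's pose, then chains per-bone Anim Dynamics with angular limits and partial alphas on side bangs and ponytails.
# A small additive bone-space offset nudges the glasses before output.
# Second-ribbon dynamics, glasses Look At and RigidBody nodes are disconnected; the event graph has no logic.
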AnimBlueprint /BRCosmetics/Characters/Player/Female/Medium/Bodies/F_MED_Street_Goth/Meshes/F_MED_Street_Goth_FaceAcc_Skeleton_AnimBP
target_skeleton=&/F_MED_Street_Goth_FaceAcc_Skeleton
graphs [
 AnimGraph {nodes [
   n1 = Anim Dynamics - Bone: dyn_ribbon_r_02(Alpha=1, ExternalForce="0,0,0") {
    angularDampingOverride=0.97
    bGravityOverrideInSimSpace=true
    gravityScale=0.2
    lODThreshold=1
    linearDampingOverride=0.97
    numSolverIterationsPostUpdate=4
    numSolverIterationsPreUpdate=16
    alphaBoolBlend {blendOption=Linear}
    boundBone {boneName=dyn_ribbon_r_02}
    componentLinearAccScale=<1 1 1>
    componentLinearVelScale=<1 1 1>
    physicsBodyDefinitions [{
      sphereCollisionRadius=0
      boundBone {boneName=dyn_ribbon_r_02}
      boxExtents=<5 5 5>
      constraintSetup {angularTargetAxis=AxisZ, twistAxis=AxisZ, angularLimitsMax=<45 45 45>, angularLimitsMin=<-45 -45 -45>, angularTarget=<0 0 0>}
      localJointOffset=<0 0 -0.1>
     }]
   }
   Anim Dynamics - Bone: dyn_ribbon_l_02(ComponentPose=@n1, Alpha=1, ExternalForce="0,0,0") {
    angularDampingOverride=0.97
    bGravityOverrideInSimSpace=true
    gravityScale=0.2
    lODThreshold=1
    linearDampingOverride=0.97
    numSolverIterationsPostUpdate=4
    numSolverIterationsPreUpdate=16
    alphaBoolBlend {blendOption=Linear}
    boundBone {boneName=dyn_ribbon_l_02}
    componentLinearAccScale=<1 1 1>
    componentLinearVelScale=<1 1 1>
    physicsBodyDefinitions [{
      sphereCollisionRadius=0
      boundBone {boneName=dyn_ribbon_l_02}
      boxExtents=<5 5 5>
      constraintSetup {angularTargetAxis=AxisZ, twistAxis=AxisZ, angularLimitsMax=<45 45 45>, angularLimitsMin=<-45 -45 -45>, angularTarget=<0 0 0>}
      localJointOffset=<0 0 -0.1>
     }]
   }
   Look At - Bone: dyn_glasses(Alpha=1) {
    alphaBoolBlend {blendOption=Linear}
    boneToModify {boneName=dyn_glasses}
    lookAtLocation=<1.3 0 0>
    lookAtTarget {boneReference {boneName=head}}
    lookAt_Axis {axis=<-1 0>}
   }
   RigidBody() {
    bOverrideWorldGravity=true
    overridePhysicsAsset=&/F_MED_Street_Goth_FaceACC_Physics
    alphaBoolBlend {blendOption=Linear}
    componentLinearAccScale=<1 1 1>
    componentLinearVelScale=<1 1 1>
    overrideWorldGravity {z=-980}
   }
   n2 = Anim Dynamics - Bone: dyn_bang_c_01(ComponentPose=Local To Component(LocalPose=Copy Pose From Mesh() {bUseAttachedParent=true}), Alpha=1, ExternalForce="0,0,0") {
    angularSpringConstant=1200
    bGravityOverrideInSimSpace=true
    gravityScale=0.33
    simulationSpace=RootRelative
    alphaBoolBlend {blendOption=Linear}
    boundBone {boneName=dyn_bang_c_01}
    physicsBodyDefinitions [{
      sphereCollisionRadius=0
      boundBone {boneName=dyn_bang_c_01}
      boxExtents=<30 30 2>
      constraintSetup {angularTargetAxis=AxisZ, twistAxis=AxisZ, angularLimitsMax=<10 0 0>, angularLimitsMin=<-30 0 0>, angularTarget=<0 0 1>}
      localJointOffset=<0 -0.1 -15>
     }]
   }
   n3 = Anim Dynamics - Bone: dyn_bang_c_02(ComponentPose=@n2, Alpha=1, ExternalForce="0,0,0") {
    angularSpringConstant=1500
    bGravityOverrideInSimSpace=true
    gravityScale=0.33
    simulationSpace=RootRelative
    alphaBoolBlend {blendOption=Linear}
    boundBone {boneName=dyn_bang_c_02}
    physicsBodyDefinitions [{
      sphereCollisionRadius=0
      boundBone {boneName=dyn_bang_c_02}
      boxExtents=<30 20 2>
      constraintSetup {angularTargetAxis=AxisZ, twistAxis=AxisZ, angularLimitsMax=<10 0 0>, angularLimitsMin=<-20 0 0>, angularTarget=<0 0 1>}
      localJointOffset=<0 -0.1 -20>
     }]
   }
   n4 = Anim Dynamics - Bone: dyn_bang_l_01(ComponentPose=@n3, Alpha=0.125, ExternalForce="0,0,0") {
    angularDampingOverride=0.98
    angularSpringConstant=1500
    bGravityOverrideInSimSpace=true
    bOverrideAngularBias=true
    bOverrideAngularDamping=true
    gravityScale=0.33
    linearDampingOverride=0.98
    numSolverIterationsPostUpdate=4
    numSolverIterationsPreUpdate=16
    simulationSpace=RootRelative
    alphaBoolBlend {blendOption=Linear}
    boundBone {boneName=dyn_bang_l_01}
    componentLinearAccScale=<1 1 1>
    componentLinearVelScale=<1 1 1>
    physicsBodyDefinitions [{
      sphereCollisionRadius=0
      boundBone {boneName=dyn_bang_l_01}
      boxExtents=<30 30 2>
      constraintSetup {angularLimitsMax=<30 80 0>, angularLimitsMin=<-30 -80 0>, angularTarget=<0 0 1>}
      localJointOffset=<0 0 17>
     }]
   }
   n5 = Anim Dynamics - Bone: dyn_bang_l_02(ComponentPose=@n4, Alpha=0.125, ExternalForce="0,0,0") {
    angularDampingOverride=0.98
    angularSpringConstant=1500
    bGravityOverrideInSimSpace=true
    bOverrideAngularBias=true
    bOverrideAngularDamping=true
    gravityScale=0.33
    linearDampingOverride=0.98
    numSolverIterationsPostUpdate=4
    numSolverIterationsPreUpdate=16
    simulationSpace=RootRelative
    alphaBoolBlend {blendOption=Linear}
    boundBone {boneName=dyn_bang_l_02}
    componentLinearAccScale=<1 1 1>
    componentLinearVelScale=<1 1 1>
    physicsBodyDefinitions [{
      sphereCollisionRadius=0
      boundBone {boneName=dyn_bang_l_02}
      boxExtents=<30 30 2>
      constraintSetup {angularLimitsMax=<30 80 0>, angularLimitsMin=<-30 -80 0>, angularTarget=<0 0 1>}
      localJointOffset=<0 0 17>
     }]
   }
   n6 = Anim Dynamics - Bone: dyn_bang_l_03(ComponentPose=@n5, Alpha=0.125, ExternalForce="0,0,0") {
    angularDampingOverride=0.98
    angularSpringConstant=1500
    bGravityOverrideInSimSpace=true
    bOverrideAngularBias=true
    bOverrideAngularDamping=true
    gravityScale=0.33
    linearDampingOverride=0.98
    numSolverIterationsPostUpdate=4
    numSolverIterationsPreUpdate=16
    simulationSpace=RootRelative
    alphaBoolBlend {blendOption=Linear}
    boundBone {boneName=dyn_bang_l_03}
    componentLinearAccScale=<1 1 1>
    componentLinearVelScale=<1 1 1>
    physicsBodyDefinitions [{
      sphereCollisionRadius=0
      boundBone {boneName=dyn_bang_l_03}
      boxExtents=<30 30 2>
      constraintSetup {angularLimitsMax=<30 80 0>, angularLimitsMin=<-30 -80 0>, angularTarget=<0 0 1>}
      localJointOffset=<0 0 17>
     }]
   }
   n7 = Anim Dynamics - Bone: dyn_bang_r_01(ComponentPose=@n6, Alpha=0.125, ExternalForce="0,0,0") {
    angularDampingOverride=0.98
    angularSpringConstant=1500
    bGravityOverrideInSimSpace=true
    bOverrideAngularBias=true
    bOverrideAngularDamping=true
    gravityScale=0.33
    linearDampingOverride=0.98
    numSolverIterationsPostUpdate=4
    numSolverIterationsPreUpdate=16
    simulationSpace=RootRelative
    alphaBoolBlend {blendOption=Linear}
    boundBone {boneName=dyn_bang_r_01}
    componentLinearAccScale=<1 1 1>
    componentLinearVelScale=<1 1 1>
    physicsBodyDefinitions [{
      sphereCollisionRadius=0
      boundBone {boneName=dyn_bang_r_01}
      boxExtents=<30 30 2>
      constraintSetup {angularLimitsMax=<30 80 0>, angularLimitsMin=<-30 -5 0>, angularTarget=<0 0 1>}
      localJointOffset=<0 0 -17>
     }]
   }
   n8 = Anim Dynamics - Bone: dyn_bang_r_02(ComponentPose=@n7, Alpha=0.125, ExternalForce="0,0,0") {
    angularDampingOverride=0.98
    angularSpringConstant=1500
    bGravityOverrideInSimSpace=true
    bOverrideAngularBias=true
    bOverrideAngularDamping=true
    gravityScale=0.33
    linearDampingOverride=0.98
    numSolverIterationsPostUpdate=4
    numSolverIterationsPreUpdate=16
    simulationSpace=RootRelative
    alphaBoolBlend {blendOption=Linear}
    boundBone {boneName=dyn_bang_r_02}
    componentLinearAccScale=<1 1 1>
    componentLinearVelScale=<1 1 1>
    physicsBodyDefinitions [{
      sphereCollisionRadius=0
      boundBone {boneName=dyn_bang_r_02}
      boxExtents=<30 30 2>
      constraintSetup {angularLimitsMax=<30 80 0>, angularLimitsMin=<-30 -5 0>, angularTarget=<0 0 1>}
      localJointOffset=<0 0 -17>
     }]
   }
   n9 = Anim Dynamics - Bone: dyn_bang_r_03(ComponentPose=@n8, Alpha=0.125, ExternalForce="0,0,0") {
    angularDampingOverride=0.98
    angularSpringConstant=1500
    bGravityOverrideInSimSpace=true
    bOverrideAngularBias=true
    bOverrideAngularDamping=true
    gravityScale=0.33
    linearDampingOverride=0.98
    numSolverIterationsPostUpdate=4
    numSolverIterationsPreUpdate=16
    simulationSpace=RootRelative
    alphaBoolBlend {blendOption=Linear}
    boundBone {boneName=dyn_bang_r_03}
    componentLinearAccScale=<1 1 1>
    componentLinearVelScale=<1 1 1>
    physicsBodyDefinitions [{
      sphereCollisionRadius=0
      boundBone {boneName=dyn_bang_r_03}
      boxExtents=<30 30 2>
      constraintSetup {angularLimitsMax=<30 80 0>, angularLimitsMin=<-30 -30 0>, angularTarget=<0 0 1>}
      localJointOffset=<0 0 -17>
     }]
   }
   n10 = Anim Dynamics - Bone: dyn_ribbon_r_01(ComponentPose=@n9, Alpha=1, ExternalForce="0,0,0") {
    angularDampingOverride=0.96
    angularSpringConstant=10000
    bGravityOverrideInSimSpace=true
    bOverrideAngularDamping=true
    gravityScale=4.8
    lODThreshold=1
    linearDampingOverride=0.96
    numSolverIterationsPostUpdate=4
    numSolverIterationsPreUpdate=16
    alphaBoolBlend {blendOption=Linear}
    boundBone {boneName=dyn_ribbon_r_01}
    componentLinearAccScale=<1 1 1>
    componentLinearVelScale=<1 1 1>
    physicsBodyDefinitions [{
      sphereCollisionRadius=0
      boundBone {boneName=dyn_ribbon_r_01}
      boxExtents=<5 5 5>
      constraintSetup {angularTargetAxis=AxisZ, twistAxis=AxisZ, angularLimitsMax=<45 45 45>, angularLimitsMin=<-45 -45 -45>, angularTarget=<0 0 0>}
      localJointOffset=<0 0 -0.1>
     }]
   }
   n11 = Anim Dynamics - Bone: dyn_ribbon_l_01(ComponentPose=@n10, Alpha=1, ExternalForce="0,0,0") {
    angularDampingOverride=0.96
    angularSpringConstant=10000
    bGravityOverrideInSimSpace=true
    bOverrideAngularDamping=true
    gravityScale=5
    lODThreshold=1
    linearDampingOverride=0.96
    numSolverIterationsPostUpdate=4
    numSolverIterationsPreUpdate=16
    alphaBoolBlend {blendOption=Linear}
    boundBone {boneName=dyn_ribbon_l_01}
    componentLinearAccScale=<0.85 0.85 0.85>
    componentLinearVelScale=<0.85 0.85 0.85>
    physicsBodyDefinitions [{
      sphereCollisionRadius=0
      boundBone {boneName=dyn_ribbon_l_01}
      boxExtents=<5 5 5>
      constraintSetup {angularTargetAxis=AxisZ, twistAxis=AxisZ, angularLimitsMax=<45 45 45>, angularLimitsMin=<-45 -45 -45>, angularTarget=<0 0 0>}
      localJointOffset=<0 0 -0.1>
     }]
   }
   n12 = Anim Dynamics - Bone: dyn_ponytail_r_02(ComponentPose=@n11, Alpha=0.75, ExternalForce="0,0,0") {
    angularDampingOverride=0.97
    bGravityOverrideInSimSpace=true
    gravityScale=0.65
    lODThreshold=1
    linearDampingOverride=0.97
    numSolverIterationsPostUpdate=4
    numSolverIterationsPreUpdate=16
    alphaBoolBlend {blendOption=Linear}
    boundBone {boneName=dyn_ponytail_r_02}
    componentLinearAccScale=<1 1 1>
    componentLinearVelScale=<1 1 1>
    physicsBodyDefinitions [{
      sphereCollisionRadius=0
      boundBone {boneName=dyn_ponytail_r_02}
      boxExtents=<5 5 5>
      constraintSetup {angularTargetAxis=AxisZ, twistAxis=AxisZ, angularLimitsMax=<45 7.5 10>, angularLimitsMin=<-45 -50 -10>, angularTarget=<0 0 0>}
      localJointOffset=<0 0 -0.1>
     }]
   }
   n13 = Anim Dynamics - Bone: dyn_ponytail_l_02(ComponentPose=@n12, Alpha=0.6, ExternalForce="0,0,0") {
    angularDampingOverride=0.97
    bGravityOverrideInSimSpace=true
    gravityScale=0.65
    lODThreshold=1
    linearDampingOverride=0.97
    numSolverIterationsPostUpdate=4
    numSolverIterationsPreUpdate=16
    alphaBoolBlend {blendOption=Linear}
    boundBone {boneName=dyn_ponytail_l_02}
    componentLinearAccScale=<0.8 0.8 0.8>
    componentLinearVelScale=<0.8 0.8 0.8>
    physicsBodyDefinitions [{
      sphereCollisionRadius=0
      boundBone {boneName=dyn_ponytail_l_02}
      boxExtents=<5 5 5>
      constraintSetup {angularTargetAxis=AxisZ, twistAxis=AxisZ, angularLimitsMax=<45 7.5 10>, angularLimitsMin=<-45 -50 -10>, angularTarget=<0 0 0>}
      localJointOffset=<0 0 -0.1>
     }]
   }
   n14 = Anim Dynamics - Bone: dyn_ponytail_r_03(ComponentPose=@n13, Alpha=0.8, ExternalForce="0,0,0") {
    angularDampingOverride=0.97
    bGravityOverrideInSimSpace=true
    gravityScale=0.65
    lODThreshold=1
    linearDampingOverride=0.97
    numSolverIterationsPostUpdate=4
    numSolverIterationsPreUpdate=16
    alphaBoolBlend {blendOption=Linear}
    boundBone {boneName=dyn_ponytail_r_03}
    componentLinearAccScale=<0.8 0.8 0.8>
    componentLinearVelScale=<0.8 0.8 0.8>
    physicsBodyDefinitions [{
      sphereCollisionRadius=0
      boundBone {boneName=dyn_ponytail_r_03}
      boxExtents=<5 5 5>
      constraintSetup {angularTargetAxis=AxisZ, twistAxis=AxisZ, angularLimitsMax=<45 0 10>, angularLimitsMin=<-45 -20 -10>, angularTarget=<0 0 0>}
      localJointOffset=<0 0 -0.1>
     }]
   }
   n15 = Anim Dynamics - Bone: dyn_ponytail_l_03(ComponentPose=@n14, Alpha=0.75, ExternalForce="0,0,0") {
    angularDampingOverride=0.97
    bGravityOverrideInSimSpace=true
    gravityScale=0.65
    lODThreshold=1
    linearDampingOverride=0.97
    numSolverIterationsPostUpdate=4
    numSolverIterationsPreUpdate=16
    alphaBoolBlend {blendOption=Linear}
    boundBone {boneName=dyn_ponytail_l_03}
    componentLinearAccScale=<1 1 1>
    componentLinearVelScale=<1 1 1>
    physicsBodyDefinitions [{
      sphereCollisionRadius=0
      boundBone {boneName=dyn_ponytail_l_03}
      boxExtents=<5 5 5>
      constraintSetup {angularTargetAxis=AxisZ, twistAxis=AxisZ, angularLimitsMax=<45 0 10>, angularLimitsMin=<-45 -20 -10>, angularTarget=<0 0 0>}
      localJointOffset=<0 0 -0.1>
     }]
   }
   n16 = "Transform (Modify) Bone - Bone: dyn_glasses"(ComponentPose=@n15, Alpha=1, Translation="0.15,0,0", Rotation="0,0,0", Scale="1,1,1") {
    translationMode=BMM_Additive
    translationSpace=BCS_BoneSpace
    alphaBoolBlend {blendOption=Linear}
    boneToModify {boneName=dyn_glasses}
    translation {x=0.15}
   }
   Output Pose(Result=Component To Local(ComponentPose=@n16)) {layerGroup=None}
  ]}
 EventGraph {nodes [
   TryGetPawnOwner()
   Event BlueprintUpdateAnimation()
  ]}
]
variables []
